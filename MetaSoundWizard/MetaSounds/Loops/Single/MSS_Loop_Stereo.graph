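# Stereo looping MetaSound Source that plays a single wave asset supplied through the Sound input, which is empty by default.
# The On Play trigger starts one stereo Wave Player, whose left and right outputs feed the source's two channels directly.
# The source is flagged as looping, though the player's Loop default holds a MetaSound builder result value in place of a plain boolean.
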MetaSoundSource /MetaSoundWizard/MetaSounds/Loops/Single/MSS_Loop_Stereo
inputs [
 Sound {type=WaveAsset}
 UE.Source.OnPlay {type=Trigger}
]
interfaces [UE.OutputFormat.Stereo:1.0, UE.Source:1.1]
outputs [
 "UE.OutputFormat.Stereo.Audio:0" {type=Audio}
 "UE.OutputFormat.Stereo.Audio:1" {type=Audio}
]
pages [Default {
  input_defaults {Sound=null, UE.Source.OnPlay=null}
  nodes [WavePlayer_0 = UE.Wave Player:Stereo(Wave Asset=^Sound, Play=^UE.Source.OnPlay) {also {defaults {Loop="<MetaSoundBuilderResult.SUCCEEDED: 0>"}}}]
  links [{dst="@UE.OutputFormat.Stereo.Audio:0", src=WavePlayer_0.Out Left}, {dst="@UE.OutputFormat.Stereo.Audio:1", src=WavePlayer_0.Out Right}]
 }]
properties {
 assetImportData=&self:AssetImportData
 bLooping=true
 duration=10000
 numChannels=2
 outputFormat=Stereo
}
